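# Physics asset for the accessory mesh of the male medium Tomcat Casual body, covering a dangling gun and left and right jacket straps.
# Capsule bodies form short two-bone chains under the pelvis, with narrow swing and twist limits between each dyn bone and its master.
# Both strap masters are angularly locked to the pelvis, while the gun master rotates freely.
# A right-thigh capsule, also locked to the pelvis, adds a leg volume.
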PhysicsAsset /BRCosmetics/Characters/Player/Male/Medium/Bodies/M_MED_Tomcat_Casual/Meshes/M_MED_Tomcat_Casual_Acc_Physics
bodies {
 dyn_gun_01 [{shapeType=Capsule, length=20.673, radius=4.7942, shapeName=dyn_gun_01_capsule, center=<14.8391 0.8231 3.2992>, rotation=<rot 90 0 0>}]
 dyn_gun_master [{shapeType=Capsule, length=2, radius=2, shapeName=dyn_gun_master_capsule, rotation=<rot 90 0 0>}]
 dyn_jacket_strap_l [{shapeType=Capsule, length=3.32, radius=0.68, shapeName=dyn_jacket_strap_l_capsule, center=<2.4995 0 0>, rotation=<rot 90 0 0>}]
 dyn_jacket_strap_l_master [{
   shapeType=Capsule
   length=4.1312
   radius=0.6751
   shapeName=dyn_jacket_strap_l_master_capsule
   center=<3.4358 0 -1.2943>
   rotation=<rot 0 0 90>
  }]
 dyn_jacket_strap_r [{shapeType=Capsule, length=3.32, radius=0.68, shapeName=dyn_jacket_strap_r_capsule, center=<-2.5178 0 -0.0001>, rotation=<rot 90 0 0>}]
 dyn_jacket_strap_r_master [{
   shapeType=Capsule
   length=4.1312
   radius=0.6751
   shapeName=dyn_jacket_strap_r_master_capsule
   center=<-3.449 0.0001 1.3777>
   rotation=<rot 0 0 90>
  }]
 pelvis [{shapeType=Capsule, length=2, radius=2, shapeName=pelvis_capsule, rotation=<rot 90 0 0>}]
 thigh_r [{shapeType=Capsule, length=28.1772, radius=6.6878, shapeName=thigh_r_capsule, center=<18.5268 -0.0001 0>, rotation=<rot 90 0 0>}]
}
constraints [
 {bone1Name=dyn_jacket_strap_l, bone2Name=dyn_jacket_strap_l_master, swing1LimitDegrees=7, swing2LimitDegrees=25, twistLimitDegrees=5}
 {bone1Name=dyn_jacket_strap_r, bone2Name=dyn_jacket_strap_r_master, swing1LimitDegrees=7, swing2LimitDegrees=25, twistLimitDegrees=5}
 {bone1Name=dyn_gun_01, bone2Name=dyn_gun_master, swing1LimitDegrees=30, swing2LimitDegrees=25, twistLimitDegrees=10}
 {
  bone1Name=dyn_jacket_strap_l_master
  bone2Name=pelvis
  swing1LimitDegrees=45
  swing1Motion=Locked
  swing2LimitDegrees=45
  swing2Motion=Locked
  twistLimitDegrees=45
  twistMotion=Locked
 }
 {
  bone1Name=dyn_jacket_strap_r_master
  bone2Name=pelvis
  swing1LimitDegrees=45
  swing1Motion=Locked
  swing2LimitDegrees=45
  swing2Motion=Locked
  twistLimitDegrees=45
  twistMotion=Locked
 }
 {
  bone1Name=dyn_gun_master
  bone2Name=pelvis
  swing1LimitDegrees=45
  swing1Motion=Free
  swing2LimitDegrees=45
  swing2Motion=Free
  twistLimitDegrees=45
  twistMotion=Free
 }
 {
  bone1Name=thigh_r
  bone2Name=pelvis
  swing1LimitDegrees=45
  swing1Motion=Locked
  swing2LimitDegrees=45
  swing2Motion=Locked
  twistLimitDegrees=45
  twistMotion=Locked
 }
]
