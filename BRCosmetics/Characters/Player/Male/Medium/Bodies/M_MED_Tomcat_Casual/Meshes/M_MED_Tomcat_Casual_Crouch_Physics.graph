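# Physics asset for the crouch variant of the Tomcat Casual male medium body, giving its jacket bones capsule bodies for physics-driven motion.
# Six two-bone dyn_jacket chains (left and right side, front and back) hang from spine_02 with individually tuned, mostly tight angular limits.
# Pelvis, lower spine and thigh capsules form the body volume, with the spine_01-to-pelvis joint left angularly free.
# Each front lower jacket bone carries two angled capsules.
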PhysicsAsset /BRCosmetics/Characters/Player/Male/Medium/Bodies/M_MED_Tomcat_Casual/Meshes/M_MED_Tomcat_Casual_Crouch_Physics
bodies {
 dyn_jacket_l_01 [{shapeType=Capsule, length=9.3966, radius=1.3286, shapeName=dyn_jacket_l_01_capsule, center=<8.948 0 0.0002>, rotation=<rot 0 0 90>}]
 dyn_jacket_l_02 [{shapeType=Capsule, length=11.6084, radius=0.8669, shapeName=dyn_jacket_l_02_capsule, center=<4.088 -0.0001 0.0002>, rotation=<rot 0 0 90>}]
 dyn_jacket_l_bk_01 [{
   shapeType=Capsule
   length=10.2174
   radius=2.2579
   shapeName=dyn_jacket_l_bk_01_capsule
   center=<9.5992 -0.0001 0.0002>
   rotation=<rot 0 0 90>
  }]
 dyn_jacket_l_bk_02 [{shapeType=Capsule, length=11.272, radius=1.2033, shapeName=dyn_jacket_l_bk_02_capsule, center=<2.4421 -0.0001 0.0002>, rotation=<rot 0 0 90>}]
 dyn_jacket_l_fr_01 [{shapeType=Capsule, length=10.7166, radius=1.7587, shapeName=dyn_jacket_l_fr_01_capsule, center=<7.0657 0 0.0001>, rotation=<rot 0 0 90>}]
 dyn_jacket_l_fr_02 [
  {
   shapeType=Capsule
   length=17.5248
   radius=0.8996
   shapeName=dyn_jacket_l_fr_02_capsule
   center=<6.4743 -4.5616 -0.7357>
   rotation=<rot 0 0 64.4865>
  }
  {
   shapeType=Capsule
   length=17.1856
   radius=0.8996
   shapeName=dyn_jacket_l_fr_02_capsule
   center=<6.5301 -12.4616 -8.114>
   rotation=<rot 0 0 17.4888>
  }
 ]
 dyn_jacket_r_01 [{shapeType=Capsule, length=11.271, radius=1.2043, shapeName=dyn_jacket_r_01_capsule, center=<9.5992 -0.0001 0.0002>, rotation=<rot 0 0 90>}]
 dyn_jacket_r_02 [{shapeType=Capsule, length=11.272, radius=1.2033, shapeName=dyn_jacket_r_02_capsule, center=<2.4421 -0.0001 0.0002>, rotation=<rot 0 0 90>}]
 dyn_jacket_r_bk_01 [{shapeType=Capsule, length=11.021, radius=1.4543, shapeName=dyn_jacket_r_bk_01_capsule, center=<9.5992 -0.0001 0.0002>, rotation=<rot 0 0 90>}]
 dyn_jacket_r_bk_02 [{shapeType=Capsule, length=11.272, radius=1.2033, shapeName=dyn_jacket_r_bk_02_capsule, center=<2.4421 -0.0001 0.0002>, rotation=<rot 0 0 90>}]
 dyn_jacket_r_fr_01 [{
   shapeType=Capsule
   length=10.7174
   radius=1.7579
   shapeName=dyn_jacket_r_fr_01_capsule
   center=<9.5992 -0.0001 0.0002>
   rotation=<rot 0 0 90>
  }]
 dyn_jacket_r_fr_02 [
  {
   shapeType=Capsule
   length=17.0248
   radius=1.3996
   shapeName=dyn_jacket_l_fr_02_capsule
   center=<5.5362 4.5616 0.7358>
   rotation=<rot 0 0 115.499>
  }
  {
   shapeType=Capsule
   length=16.6856
   radius=1.3996
   shapeName=dyn_jacket_l_fr_02_capsule
   center=<6.7896 12.4616 -8.114>
   rotation=<rot 0 0 162.4999>
  }
 ]
 pelvis [{shapeType=Capsule, length=6.9501, radius=10.0864}]
 spine_01 [{shapeType=Capsule, length=4.3007, radius=8.9092}]
 spine_02 [{shapeType=Capsule, length=4.3007, radius=8.9092}]
 thigh_l [{shapeType=Capsule, length=28.1337, radius=7.75, center=<-20.7322 0.0011 0>, rotation=<rot 90 0 0>}]
 thigh_r [{shapeType=Capsule, length=27.0489, radius=8.6993, center=<20.3531 0.0016 0>, rotation=<rot 90 0 0>}]
}
constraints [
 {bone1Name=spine_02, bone2Name=spine_01, swing1LimitDegrees=45, swing2LimitDegrees=45, twistLimitDegrees=45}
 {bone1Name=thigh_l, bone2Name=pelvis, swing1LimitDegrees=45, swing2LimitDegrees=45, twistLimitDegrees=45}
 {bone1Name=thigh_r, bone2Name=pelvis, swing1LimitDegrees=45, swing2LimitDegrees=45, twistLimitDegrees=45}
 {bone1Name=dyn_jacket_l_02, bone2Name=dyn_jacket_l_01, swing1LimitDegrees=7, swing2LimitDegrees=40, twistLimitDegrees=15}
 {bone1Name=dyn_jacket_l_01, bone2Name=spine_02, swing1LimitDegrees=7, swing2LimitDegrees=20, twistLimitDegrees=45}
 {bone1Name=dyn_jacket_r_02, bone2Name=dyn_jacket_r_01, swing1LimitDegrees=15, swing2LimitDegrees=20, twistLimitDegrees=5}
 {bone1Name=dyn_jacket_r_01, bone2Name=spine_02, swing1LimitDegrees=15, swing2LimitDegrees=5, twistLimitDegrees=3}
 {bone1Name=dyn_jacket_r_fr_02, bone2Name=dyn_jacket_r_fr_01, swing1LimitDegrees=7, swing2LimitDegrees=20, twistLimitDegrees=5}
 {bone1Name=dyn_jacket_r_fr_01, bone2Name=spine_02, swing1LimitDegrees=12, swing2LimitDegrees=25, twistLimitDegrees=5}
 {bone1Name=dyn_jacket_l_fr_02, bone2Name=dyn_jacket_l_fr_01, swing1LimitDegrees=7, swing2LimitDegrees=20, twistLimitDegrees=7}
 {bone1Name=dyn_jacket_l_fr_01, bone2Name=spine_02, swing1LimitDegrees=20, swing2LimitDegrees=50, twistLimitDegrees=10}
 {bone1Name=dyn_jacket_l_bk_02, bone2Name=dyn_jacket_l_bk_01, swing1LimitDegrees=7, swing2LimitDegrees=20, twistLimitDegrees=5}
 {bone1Name=dyn_jacket_l_bk_01, bone2Name=spine_02, swing1LimitDegrees=7, swing2LimitDegrees=20, twistLimitDegrees=5}
 {bone1Name=dyn_jacket_r_bk_02, bone2Name=dyn_jacket_r_bk_01, swing1LimitDegrees=7, swing2LimitDegrees=20, twistLimitDegrees=5}
 {bone1Name=dyn_jacket_r_bk_01, bone2Name=spine_02, swing1LimitDegrees=7, swing2LimitDegrees=20, twistLimitDegrees=5}
 {
  bone1Name=spine_01
  bone2Name=pelvis
  swing1LimitDegrees=45
  swing1Motion=Free
  swing2LimitDegrees=45
  swing2Motion=Free
  twistLimitDegrees=45
  twistMotion=Free
 }
]
