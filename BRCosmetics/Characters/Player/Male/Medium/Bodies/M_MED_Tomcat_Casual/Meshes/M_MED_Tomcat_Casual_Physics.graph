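# Physics asset for the male medium Tomcat Casual player body, set up for secondary motion of the jacket.
# Six two-bone capsule chains (left and right side, front and back) hang from spine_02 with tight twist and asymmetric swing limits, widest at the front roots.
# Pelvis, lower spine and thigh capsules form a simplified body around them.
# Front root bones carry two overlapping capsules, and the spine_01-to-pelvis joint rotates freely.
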PhysicsAsset /BRCosmetics/Characters/Player/Male/Medium/Bodies/M_MED_Tomcat_Casual/Meshes/M_MED_Tomcat_Casual_Physics
bodies {
 dyn_jacket_l_01 [{shapeType=Capsule, length=10.4674, radius=2.0079, shapeName=dyn_jacket_l_01_capsule, center=<9.5992 -0.0001 0.0002>, rotation=<rot 0 0 90>}]
 dyn_jacket_l_02 [{shapeType=Capsule, length=11.272, radius=1.2033, shapeName=dyn_jacket_l_02_capsule, center=<2.4421 -0.0001 0.0002>, rotation=<rot 0 0 90>}]
 dyn_jacket_l_bk_01 [{
   shapeType=Capsule
   length=10.4674
   radius=2.0079
   shapeName=dyn_jacket_l_bk_01_capsule
   center=<9.5992 -0.0001 0.0002>
   rotation=<rot 0 0 90>
  }]
 dyn_jacket_l_bk_02 [{shapeType=Capsule, length=11.272, radius=1.2033, shapeName=dyn_jacket_l_bk_02_capsule, center=<2.4421 -0.0001 0.0002>, rotation=<rot 0 0 90>}]
 dyn_jacket_l_fr_01 [
  {
   shapeType=Capsule
   length=10.4674
   radius=2.0079
   shapeName=dyn_jacket_l_fr_01_capsule
   center=<9.5992 -0.0001 0.0002>
   rotation=<rot 0 0 90>
  }
  {shapeType=Capsule, length=8.7371, radius=2.0079, shapeName=dyn_jacket_l_fr_01_capsule, center=<6.0421 -0.0001 0.0005>, rotation=<rot 0 0 90>}
 ]
 dyn_jacket_l_fr_02 [{shapeType=Capsule, length=18.022, radius=1.6, shapeName=dyn_jacket_l_fr_02_capsule, center=<6.4472 0 0>, rotation=<rot 0 0 90>}]
 dyn_jacket_r_01 [{shapeType=Capsule, length=10.4674, radius=2.0079, shapeName=dyn_jacket_r_01_capsule, center=<9.5992 -0.0001 0.0002>, rotation=<rot 0 0 90>}]
 dyn_jacket_r_02 [{shapeType=Capsule, length=11.272, radius=1.2033, shapeName=dyn_jacket_r_02_capsule, center=<2.4421 -0.0001 0.0002>, rotation=<rot 0 0 90>}]
 dyn_jacket_r_bk_01 [{
   shapeType=Capsule
   length=10.4674
   radius=2.0079
   shapeName=dyn_jacket_r_bk_01_capsule
   center=<9.5992 -0.0001 0.0002>
   rotation=<rot 0 0 90>
  }]
 dyn_jacket_r_bk_02 [{shapeType=Capsule, length=11.272, radius=1.2033, shapeName=dyn_jacket_r_bk_02_capsule, center=<2.4421 -0.0001 0.0002>, rotation=<rot 0 0 90>}]
 dyn_jacket_r_fr_01 [
  {
   shapeType=Capsule
   length=10.4674
   radius=2.0079
   shapeName=dyn_jacket_r_fr_01_capsule
   center=<9.5992 -0.0001 0.0002>
   rotation=<rot 0 0 90>
  }
  {shapeType=Capsule, length=8.3762, radius=2.0079, shapeName=dyn_jacket_r_fr_01_capsule, center=<5.9504 -0.0001 0.0004>, rotation=<rot 0 0 90>}
 ]
 dyn_jacket_r_fr_02 [{shapeType=Capsule, length=18.022, radius=1.6, shapeName=dyn_jacket_r_fr_02_capsule, center=<6.4472 0 0>, rotation=<rot 0 0 90>}]
 pelvis [{shapeType=Capsule, length=7.2001, radius=10.0864}]
 spine_01 [{shapeType=Capsule, length=4.3007, radius=8.9092}]
 spine_02 [{shapeType=Capsule, length=4.3007, radius=8.9092}]
 thigh_l [{shapeType=Capsule, length=28.0489, radius=8.9493, center=<-20.7322 0.0011 0>, rotation=<rot 90 0 0>}]
 thigh_r [{shapeType=Capsule, length=28.0489, radius=8.9493, center=<20.3531 0.0016 0>, rotation=<rot 90 0 0>}]
}
constraints [
 {bone1Name=spine_02, bone2Name=spine_01, swing1LimitDegrees=45, swing2LimitDegrees=45, twistLimitDegrees=45}
 {bone1Name=thigh_l, bone2Name=pelvis, swing1LimitDegrees=45, swing2LimitDegrees=45, twistLimitDegrees=45}
 {bone1Name=thigh_r, bone2Name=pelvis, swing1LimitDegrees=45, swing2LimitDegrees=45, twistLimitDegrees=45}
 {bone1Name=dyn_jacket_l_02, bone2Name=dyn_jacket_l_01, swing1LimitDegrees=7, swing2LimitDegrees=20, twistLimitDegrees=5}
 {bone1Name=dyn_jacket_l_01, bone2Name=spine_02, swing1LimitDegrees=7, swing2LimitDegrees=20, twistLimitDegrees=5}
 {bone1Name=dyn_jacket_r_02, bone2Name=dyn_jacket_r_01, swing1LimitDegrees=7, swing2LimitDegrees=20, twistLimitDegrees=5}
 {bone1Name=dyn_jacket_r_01, bone2Name=spine_02, swing1LimitDegrees=7, swing2LimitDegrees=10, twistLimitDegrees=5}
 {bone1Name=dyn_jacket_r_fr_02, bone2Name=dyn_jacket_r_fr_01, swing1LimitDegrees=7, swing2LimitDegrees=25, twistLimitDegrees=5}
 {bone1Name=dyn_jacket_r_fr_01, bone2Name=spine_02, swing1LimitDegrees=7, swing2LimitDegrees=50, twistLimitDegrees=5}
 {bone1Name=dyn_jacket_l_fr_02, bone2Name=dyn_jacket_l_fr_01, swing1LimitDegrees=7, swing2LimitDegrees=25, twistLimitDegrees=5}
 {bone1Name=dyn_jacket_l_fr_01, bone2Name=spine_02, swing1LimitDegrees=7, swing2LimitDegrees=50, twistLimitDegrees=5}
 {bone1Name=dyn_jacket_l_bk_02, bone2Name=dyn_jacket_l_bk_01, swing1LimitDegrees=7, swing2LimitDegrees=20, twistLimitDegrees=5}
 {bone1Name=dyn_jacket_l_bk_01, bone2Name=spine_02, swing1LimitDegrees=7, swing2LimitDegrees=20, twistLimitDegrees=5}
 {bone1Name=dyn_jacket_r_bk_02, bone2Name=dyn_jacket_r_bk_01, swing1LimitDegrees=7, swing2LimitDegrees=20, twistLimitDegrees=5}
 {bone1Name=dyn_jacket_r_bk_01, bone2Name=spine_02, swing1LimitDegrees=7, swing2LimitDegrees=20, twistLimitDegrees=5}
 {
  bone1Name=spine_01
  bone2Name=pelvis
  swing1LimitDegrees=45
  swing1Motion=Free
  swing2LimitDegrees=45
  swing2Motion=Free
  twistLimitDegrees=45
  twistMotion=Free
 }
]
