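# Anim Blueprint for the male Tomcat Casual body that copies the parent mesh pose and adds secondary motion to jacket, straps and gun.
# Crouch-switched RigidBody branches, one with Control Rig and corrections, feed weighted constraints and accessory physics, cached as Pre_Jump.
# Sprint, airborne and skydive bone offsets follow, then forearm-twist constraints and a lapel Control Rig.
# EventGraph converts movement state and direction into corrective alphas, and resets dynamics shortly after emotes end.
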AnimBlueprint /BRCosmetics/Characters/Player/Male/Medium/Bodies/M_MED_Tomcat_Casual/Meshes/M_MED_Tomcat_Casual_AnimBP
target_skeleton=&/M_MED_Tomcat_Casual_Skeleton
graphs [
 AnimGraph {nodes [
   "Constraint - dyn_gun_master (P)"(Alpha=0.7, ConstraintWeights_0=1) {
    alphaBoolBlend {blendOption=Linear}
    boneToModify {boneName=dyn_gun_master}
    constraintSetup [{targetBone {boneName=thigh_r}}]
    constraintWeights [1]
   }
   n1 = "Constraint - dyn_jacket_strap_r_master (PPPPPPP)"!(
    ComponentPose
    Alpha=1
    ConstraintWeights_0=0.3259
    ConstraintWeights_1:float=0.0871
    ConstraintWeights_2:float=0.1256
    ConstraintWeights_3:float=0.1948
    ConstraintWeights_4:float=0.1625
    ConstraintWeights_5:float=0.0465
    ConstraintWeights_6:float=0.0577
   ) -> (Pose) {
    alphaBoolBlend {blendOption=Linear}
    boneToModify {boneName=dyn_jacket_strap_r_master}
    constraintSetup [
     {targetBone {boneName=dyn_jacket_r_fr_01}}
     {targetBone {boneName=dyn_jacket_r_01}}
     {targetBone {boneName=spine_01}}
     {targetBone {boneName=spine_02}}
     {targetBone {boneName=spine_03}}
     {targetBone {boneName=spine_04}}
     {targetBone {boneName=pelvis}}
    ]
    constraintWeights [1,1,1,1,1,1,1]
   }
   "Constraint - dyn_jacket_strap_l_master (PPPPPPP)"!(
    ComponentPose=@n1
    Alpha=1
    ConstraintWeights_0=0.1459
    ConstraintWeights_1:float=0.102
    ConstraintWeights_2:float=0.1587
    ConstraintWeights_3:float=0.1325
    ConstraintWeights_4:float=0.0376
    ConstraintWeights_5:float=0.0476
    ConstraintWeights_6:float=0.3759
   ) -> (Pose) {
    alphaBoolBlend {blendOption=Linear}
    boneToModify {boneName=dyn_jacket_strap_l_master}
    constraintSetup [
     {targetBone {boneName=dyn_jacket_l_01}}
     {targetBone {boneName=spine_01}}
     {targetBone {boneName=spine_02}}
     {targetBone {boneName=spine_03}}
     {targetBone {boneName=spine_04}}
     {targetBone {boneName=pelvis}}
     {targetBone {boneName=dyn_jacket_l_fr_01}}
    ]
    constraintWeights [1,1,1,1,1,1,1]
   }
   "Constraint - dyn_jacket_r_fr_01 (R)"(Alpha=0.3, ConstraintWeights_0=1) {
    alphaBoolBlend {blendOption=Linear}
    boneToModify {boneName=dyn_jacket_r_fr_01}
    constraintSetup [{transformType=Rotation, targetBone {boneName=thigh_twist_01_r}}]
    constraintWeights [1]
   }
   "Transform (Modify) Bone - Bone: dyn_jacket_l_master"(Alpha=1, Translation="0,0,0", Rotation="-15,0,0", Scale="1,1,1") {
    rotationMode=BMM_Additive
    translationMode=BMM_Additive
    alphaBoolBlend {blendOption=Linear}
    boneToModify {boneName=dyn_jacket_l_master}
    rotation {pitch=-15}
   }
   n2 = "Constraint - dyn_gun_master (P)"(ComponentPose=Local To Component(LocalPose=Use cached pose 'Pre_Jump'()), Alpha=0.7, ConstraintWeights_0=1) {
    alphaBoolBlend {blendOption=Linear}
    boneToModify {boneName=dyn_gun_master}
    constraintSetup [{targetBone {boneName=thigh_r}}]
    constraintWeights [1]
   }
   n3 = "Transform (Modify) Bone - Bone: dyn_jacket_r_master"(ComponentPose=@n2, Alpha=1, Translation="0,0,1.5", Rotation="0,0,0", Scale="1,1,1") {
    rotationMode=BMM_Additive
    translationMode=BMM_Additive
    alphaBoolBlend {blendOption=Linear}
    boneToModify {boneName=dyn_jacket_r_master}
    translation {z=1.5}
   }
   n4 = "Transform (Modify) Bone - Bone: dyn_jacket_r_fr_master"(ComponentPose=@n3, Alpha=1, Translation="-3,0,0", Rotation="0,0,5", Scale="1,1,1") {
    rotationMode=BMM_Additive
    rotationSpace=BCS_BoneSpace
    translationMode=BMM_Additive
    translationSpace=BCS_BoneSpace
    alphaBoolBlend {blendOption=Linear}
    boneToModify {boneName=dyn_jacket_r_fr_master}
    rotation {roll=5}
    translation {x=-3}
   }
   n5 = "Transform (Modify) Bone - Bone: dyn_jacket_r_fr_02"(ComponentPose=@n4, Alpha=1, Translation="-3,0,0", Rotation="0,0,0", Scale="1,1,1") {
    rotationMode=BMM_Additive
    rotationSpace=BCS_BoneSpace
    translationMode=BMM_Additive
    translationSpace=BCS_BoneSpace
    alphaBoolBlend {blendOption=Linear}
    boneToModify {boneName=dyn_jacket_r_fr_02}
    translation {x=-3}
   }
   n6 = "Transform (Modify) Bone - Bone: dyn_gun_master"(ComponentPose=@n5, Alpha=1, Translation="0,0,-3", Rotation="0,0,0", Scale="1,1,1") {
    rotationMode=BMM_Additive
    translationMode=BMM_Additive
    alphaBoolBlend {blendOption=Linear}
    boneToModify {boneName=dyn_gun_master}
    translation {z=-3}
   }
   n7 = "Transform (Modify) Bone - Bone: dyn_gun_01"(ComponentPose=Local To Component(LocalPose=Use cached pose 'Pre_Jump'()), Alpha=1, Translation="-1,0,0", Rotation="2,0,0", Scale="1,1,1") {
    rotationMode=BMM_Additive
    translationMode=BMM_Additive
    alphaBoolBlend {blendOption=Linear}
    boneToModify {boneName=dyn_gun_01}
    rotation {pitch=2}
    translation {x=-1}
   }
   n8 = Blend Poses by bool(
    BlendPose_0=Component To Local(ComponentPose=@n6)
    BlendPose_1=Component To Local(ComponentPose=@n7)
    BlendTime_1=0.25
    bActiveValue=OR Boolean(A=Get bIsJumping(), B=Get bIsFalling())
   ) {blendType=Linear}
   n9 = "Transform (Modify) Bone - Bone: dyn_jacket_r_master"(ComponentPose=Local To Component(LocalPose=Use cached pose 'Pre_Jump'()), Alpha=1, Translation="0,0,0.5", Rotation="0,0,0", Scale="1,1,1") {
    rotationMode=BMM_Additive
    translationMode=BMM_Additive
    alphaBoolBlend {blendOption=Linear}
    boneToModify {boneName=dyn_jacket_r_master}
    translation {z=0.5}
   }
   n10 = "Transform (Modify) Bone - Bone: dyn_jacket_r_fr_master"(ComponentPose=@n9, Alpha=1, Translation="-2,0,0", Rotation="0,0,0", Scale="1,1,1") {
    rotationMode=BMM_Additive
    rotationSpace=BCS_BoneSpace
    translationMode=BMM_Additive
    translationSpace=BCS_BoneSpace
    alphaBoolBlend {blendOption=Linear}
    boneToModify {boneName=dyn_jacket_r_fr_master}
    translation {x=-2}
   }
   n11 = "Transform (Modify) Bone - Bone: dyn_jacket_r_fr_02"(ComponentPose=@n10, Alpha=1, Translation="-1,0,0", Rotation="0,0,0", Scale="1,1,1") {
    rotationMode=BMM_Additive
    rotationSpace=BCS_BoneSpace
    translationMode=BMM_Additive
    translationSpace=BCS_BoneSpace
    alphaBoolBlend {blendOption=Linear}
    boneToModify {boneName=dyn_jacket_r_fr_02}
    translation {x=-1}
   }
   n12 = "Transform (Modify) Bone - Bone: dyn_gun_master"(ComponentPose=@n11, Alpha=1, Translation="-2,0,5", Rotation="0,0,0", Scale="1,1,1") {
    rotationMode=BMM_Additive
    translationMode=BMM_Additive
    alphaBoolBlend {blendOption=Linear}
    boneToModify {boneName=dyn_gun_master}
    translation {x=-2, z=5}
   }
   n13 = Component To Local(ComponentPose=@n12)
   n14 = "Transform (Modify) Bone - Bone: dyn_gun_01"(ComponentPose=Local To Component(LocalPose=Use cached pose 'Pre_Jump'()), Alpha=1, Translation="-1,0,0", Rotation="2,0,0", Scale="1,1,1") {
    rotationMode=BMM_Additive
    translationMode=BMM_Additive
    alphaBoolBlend {blendOption=Linear}
    boneToModify {boneName=dyn_gun_01}
    rotation {pitch=2}
    translation {x=-1}
   }
   n15 = Component To Local(ComponentPose=@n14)
   n16 = OR Boolean(A=Get bIsJumping(), B=Get bIsFalling())
   n17 = "Transform (Modify) Bone - Bone: dyn_gun_01"(
    ComponentPose=Local To Component(LocalPose=Blend Poses by bool(BlendPose_0=@n13, BlendPose_1=@n15, BlendTime_1=0.25, bActiveValue=@n16) {blendType=Linear})
    Alpha=1
    Translation="0,0,-3"
    Rotation="0,0,-25"
    Scale="1,1,1"
   ) {
    rotationMode=BMM_Additive
    translationMode=BMM_Additive
    alphaBoolBlend {blendOption=Linear}
    boneToModify {boneName=dyn_gun_01}
    rotation {roll=-25}
    translation {z=-3}
   }
   n18 = "Transform (Modify) Bone - Bone: dyn_jacket_r_master"(ComponentPose=@n17, Alpha=Get not crouch sprinting() -> (not crouch sprinting:double), Translation="0,0,2", Rotation="0,0,0", Scale="1,1,1") {
    rotationMode=BMM_Additive
    translationMode=BMM_Additive
    alphaBoolBlend {blendOption=Linear}
    boneToModify {boneName=dyn_jacket_r_master}
    translation {z=2}
   }
   n19 = "Transform (Modify) Bone - Bone: dyn_jacket_r_master"(ComponentPose=@n18, Alpha=Get crouch_jacket() -> (crouch_jacket:double), Translation="0,0,0", Rotation="-10,0,0", Scale="1,1,1") {
    rotationMode=BMM_Additive
    translationMode=BMM_Additive
    alphaBoolBlend {blendOption=Linear}
    boneToModify {boneName=dyn_jacket_r_master}
    rotation {pitch=-10}
   }
   n20 = Component To Local(ComponentPose=@n19)
   n21 = Get bIsSprinting()
   n22 = Get diving() -> (diving:double)
   n23 = "Transform (Modify) Bone - Bone: dyn_gun_01"(
    ComponentPose=Local To Component(LocalPose=Blend Poses by bool(BlendPose_0=@n20, BlendPose_1=@n8, BlendTime_0=0.25, BlendTime_1=0.25, bActiveValue=@n21) {blendType=Linear})
    Alpha=@n22
    Translation="0,0,-3"
    Rotation="0,0,-25"
    Scale="1,1,1"
   ) {
    rotationMode=BMM_Additive
    translationMode=BMM_Additive
    alphaBoolBlend {blendOption=Linear}
    boneToModify {boneName=dyn_gun_01}
    rotation {roll=-25}
    translation {z=-3}
   }
   n24 = "Transform (Modify) Bone - Bone: dyn_jacket_r_master"(ComponentPose=@n23, Alpha=@n22, Translation="0,0,1", Rotation="-10,0,0", Scale="1,1,1") {
    rotationMode=BMM_Additive
    translationMode=BMM_Additive
    alphaBoolBlend {blendOption=Linear}
    boneToModify {boneName=dyn_jacket_r_master}
    rotation {pitch=-10}
    translation {z=1}
   }
   n25 = Get is_crouch_walking() -> (is_crouch_walking:double)
   n26 = "Transform (Modify) Bone - Bone: dyn_gun_01"(
    ComponentPose=@n24
    Alpha=SelectFloat(
     A=0
     B=Get is_crouching() -> (is_crouching:double)
     bPickA="Equal (Float)"!(A:double=@n25, B:double=1) -> (ReturnValue:bool) {is=title K2Node_CallFunction}
    )
    Translation="-1,0,0"
    Rotation="4,10,0"
    Scale="1,1,1"
   ) {
    rotationMode=BMM_Additive
    translationMode=BMM_Additive
    alphaBoolBlend {blendOption=Linear}
    boneToModify {boneName=dyn_gun_01}
    rotation {pitch=4, yaw=10}
    translation {x=-1}
   }
   n27 = "Constraint - lowerarm_twist_01_l (P)"(ComponentPose=@n26, Alpha=1, ConstraintWeights_0=1) {
    alphaBoolBlend {blendOption=Linear}
    alphaScaleBiasClamp {bInterpResult=true, interpSpeedDecreasing=2}
    boneToModify {boneName=lowerarm_twist_01_l}
    constraintSetup [{targetBone {boneName=lowerarm_l}}]
    constraintWeights [1]
   }
   n28 = "Constraint - lowerarm_twist_02_l (P)"(ComponentPose=@n27, Alpha=1, ConstraintWeights_0=1) {
    alphaBoolBlend {blendOption=Linear}
    alphaScaleBiasClamp {bInterpResult=true, interpSpeedDecreasing=2}
    boneToModify {boneName=lowerarm_twist_02_l}
    constraintSetup [{targetBone {boneName=lowerarm_l}}]
    constraintWeights [1]
   }
   n29 = "Constraint - lowerarm_twist_01_r (P)"(ComponentPose=@n28, Alpha=1, ConstraintWeights_0=1) {
    alphaBoolBlend {blendOption=Linear}
    alphaScaleBiasClamp {bInterpResult=true, interpSpeedDecreasing=2}
    boneToModify {boneName=lowerarm_twist_01_r}
    constraintSetup [{targetBone {boneName=lowerarm_r}}]
    constraintWeights [1]
   }
   n30 = "Constraint - lowerarm_twist_02_r (P)"(ComponentPose=@n29, Alpha=1, ConstraintWeights_0=1) {
    alphaBoolBlend {blendOption=Linear}
    alphaScaleBiasClamp {bInterpResult=true, interpSpeedDecreasing=2}
    boneToModify {boneName=lowerarm_twist_02_r}
    constraintSetup [{targetBone {boneName=lowerarm_r}}]
    constraintWeights [1]
   }
   n31 = Control Rig(Source=Component To Local(ComponentPose=@n30)) {
    alphaBoolBlend {blendOption=Linear}
    controlRigAssetReference {blueprintRigClass=&/M_MED_Tomcat_Casual_CtrlRig_Lapel.M_MED_Tomcat_Casual_CtrlRig_Lapel_C}
   }
   Output Pose(Result=@n31) {layerGroup=None}
   n32 = RigidBody(ComponentPose=Local To Component(LocalPose=Copy Pose From Mesh() {bUseAttachedParent=true})) {
    bClampLinearTranslationLimitToRefPose=true
    bOverrideWorldGravity=true
    overridePhysicsAsset=&/M_MED_Tomcat_Casual_Physics
    simulationSpace=BaseBoneSpace
    alphaBoolBlend {blendOption=Linear}
    baseBoneRef {boneName=pelvis}
    componentAppliedLinearAccClamp=<3000 3000 3000>
    componentLinearAccScale=<0.3 0.3 0.4>
    componentLinearVelScale=<0.3 0.3 0.33>
    overrideWorldGravity {z=-980}
   }
   n33 = Component To Local(ComponentPose=@n32)
   n34 = RigidBody(ComponentPose=Local To Component(LocalPose=Copy Pose From Mesh() {bUseAttachedParent=true})) {
    bClampLinearTranslationLimitToRefPose=true
    bOverrideWorldGravity=true
    overridePhysicsAsset=&/M_MED_Tomcat_Casual_Crouch_Physics
    simulationSpace=BaseBoneSpace
    alphaBoolBlend {blendOption=Linear}
    baseBoneRef {boneName=pelvis}
    componentAppliedLinearAccClamp=<3000 3000 3000>
    componentLinearAccScale=<0.3 0.3 0.4>
    componentLinearVelScale=<0.3 0.3 0.33>
    overrideWorldGravity {z=-980}
   }
   n35 = Control Rig(Source=Component To Local(ComponentPose=@n34)) {
    alphaBoolBlend {blendOption=Linear}
    controlRigAssetReference {blueprintRigClass=&/M_MED_Tomcat_Casual_CtrlRig.M_MED_Tomcat_Casual_CtrlRig_C}
   }
   n36 = Get is_crouch_walking()
   n37 = "Constraint - dyn_jacket_r_fr_01 (R)"(
    ComponentPose=Local To Component(LocalPose=@n35)
    Alpha=SelectFloat(A=0.4, B=0.15, bPickA="Equal (Float)"(A=@n36, B=1))
    ConstraintWeights_0=1
   ) {
    alphaBoolBlend {blendOption=Linear}
    boneToModify {boneName=dyn_jacket_r_fr_01}
    constraintSetup [{transformType=Rotation, targetBone {boneName=thigh_twist_01_r}}]
    constraintWeights [1]
   }
   n38 = "Constraint - dyn_jacket_r_01 (T)"(ComponentPose=@n37, Alpha=0.6, ConstraintWeights_0=1) {
    alphaBoolBlend {blendOption=Linear}
    boneToModify {boneName=dyn_jacket_r_01}
    constraintSetup [{transformType=Translation, targetBone {boneName=spine_01}}]
    constraintWeights [1]
   }
   n39 = "Constraint - dyn_jacket_l_fr_01 (T)"(ComponentPose=@n38, Alpha=0.26, ConstraintWeights_0=1) {
    alphaBoolBlend {blendOption=Linear}
    boneToModify {boneName=dyn_jacket_l_fr_01}
    constraintSetup [{transformType=Translation, targetBone {boneName=thigh_l}}]
    constraintWeights [1]
   }
   n40 = "Constraint - dyn_jacket_l_01 (R)"(ComponentPose=@n39, Alpha=0.15, ConstraintWeights_0=1) {
    alphaBoolBlend {blendOption=Linear}
    boneToModify {boneName=dyn_jacket_l_01}
    constraintSetup [{transformType=Rotation, targetBone {boneName=thigh_l}}]
    constraintWeights [1]
   }
   n41 = "Constraint - dyn_jacket_l_fr_02 (R)"(ComponentPose=@n40, Alpha=0.6, ConstraintWeights_0=1) {
    alphaBoolBlend {blendOption=Linear}
    boneToModify {boneName=dyn_jacket_l_fr_02}
    constraintSetup [{transformType=Rotation, targetBone {boneName=thigh_l}}]
    constraintWeights [1]
   }
   n42 = "Transform (Modify) Bone - Bone: dyn_jacket_l_fr_master"(ComponentPose=@n41, Alpha=1, Translation="0,0,0.7", Rotation="3,-5,0", Scale="1,1,1") {
    rotationMode=BMM_Additive
    translationMode=BMM_Additive
    alphaBoolBlend {blendOption=Linear}
    boneToModify {boneName=dyn_jacket_l_fr_master}
    rotation {pitch=3, yaw=-5}
    translation {z=0.7}
   }
   n43 = "Transform (Modify) Bone - Bone: dyn_jacket_r_master"(ComponentPose=@n42, Alpha=1, Translation="0.5,0,0.5", Rotation="8,0,0", Scale="1,1,1") {
    rotationMode=BMM_Additive
    translationMode=BMM_Additive
    alphaBoolBlend {blendOption=Linear}
    boneToModify {boneName=dyn_jacket_r_master}
    rotation {pitch=8}
    translation {x=0.5, z=0.5}
   }
   n44 = "Transform (Modify) Bone - Bone: dyn_jacket_r_fr_master"(ComponentPose=@n43, Alpha=1, Translation="0,0,1", Rotation="0,0,0", Scale="1,1,1") {
    rotationMode=BMM_Additive
    translationMode=BMM_Additive
    alphaBoolBlend {blendOption=Linear}
    boneToModify {boneName=dyn_jacket_r_fr_master}
    translation {z=1}
   }
   n45 = "Transform (Modify) Bone - Bone: dyn_jacket_r_master"(ComponentPose=@n44, Alpha=1, Translation="-0.5,0,0", Rotation="0,0,0", Scale="1,1,1") {
    rotationMode=BMM_Additive
    translationMode=BMM_Additive
    alphaBoolBlend {blendOption=Linear}
    boneToModify {boneName=dyn_jacket_r_master}
    translation {x=-0.5}
   }
   n46 = "Transform (Modify) Bone - Bone: dyn_jacket_r_fr_master"(ComponentPose=@n45, Alpha=Get Crouch_Bk_Left() -> (Crouch_Bk_Left:double), Translation="0,0,0.5", Rotation="-10,0,0", Scale="1,1,1") {
    rotationMode=BMM_Additive
    translationMode=BMM_Additive
    alphaBoolBlend {blendOption=Linear}
    boneToModify {boneName=dyn_jacket_r_fr_master}
    rotation {pitch=-10}
    translation {z=0.5}
   }
   n47 = "Transform (Modify) Bone - Bone: dyn_jacket_r_fr_master"(ComponentPose=@n46, Alpha=Get Crouch_Bk() -> (Crouch_Bk:double), Translation="0,0,0.5", Rotation="-12,0,0", Scale="1,1,1") {
    rotationMode=BMM_Additive
    translationMode=BMM_Additive
    alphaBoolBlend {blendOption=Linear}
    boneToModify {boneName=dyn_jacket_r_fr_master}
    rotation {pitch=-12}
    translation {z=0.5}
   }
   n48 = SelectFloat(A=1, B=0, bPickA="Equal (Float)"(A=Get is_crouch_walking(), B=1))
   n49 = "Transform (Modify) Bone - Bone: dyn_jacket_r_fr_01"(ComponentPose=@n47, Alpha=@n48, Translation="-2,0,0", Rotation="0,0,0", Scale="1,1,1") {
    rotationMode=BMM_Additive
    rotationSpace=BCS_BoneSpace
    translationMode=BMM_Additive
    translationSpace=BCS_BoneSpace
    alphaBoolBlend {blendOption=Linear}
    alphaScaleBiasClamp {bInterpResult=true}
    boneToModify {boneName=dyn_jacket_r_fr_01}
    translation {x=-2}
   }
   n50 = "Transform (Modify) Bone - Bone: dyn_jacket_r_fr_02"(ComponentPose=@n49, Alpha=@n48, Translation="-3,0,0", Rotation="0,0,0", Scale="1,1,1") {
    rotationMode=BMM_Additive
    rotationSpace=BCS_BoneSpace
    translationMode=BMM_Additive
    translationSpace=BCS_BoneSpace
    alphaBoolBlend {blendOption=Linear}
    alphaScaleBiasClamp {bInterpResult=true}
    boneToModify {boneName=dyn_jacket_r_fr_02}
    translation {x=-3}
   }
   n51 = Component To Local(ComponentPose=@n50)
   n52 = Get bIsCrouching()
   n53 = "Constraint - dyn_jacket_strap_r_master (PPPPPPP)"(
    ComponentPose=Local To Component(LocalPose=Blend Poses by bool(BlendPose_0=@n51, BlendPose_1=@n33, BlendTime_1=0.25, bActiveValue=@n52) {blendType=Linear})
    Alpha=1
    ConstraintWeights_0=0.3259
    ConstraintWeights_1=0.0871
    ConstraintWeights_2=0.1256
    ConstraintWeights_3=0.1948
    ConstraintWeights_4=0.1625
    ConstraintWeights_5=0.0465
    ConstraintWeights_6=0.0577
   ) {
    alphaBoolBlend {blendOption=Linear}
    boneToModify {boneName=dyn_jacket_strap_r_master}
    constraintSetup [
     {targetBone {boneName=dyn_jacket_r_fr_01}}
     {targetBone {boneName=dyn_jacket_r_01}}
     {targetBone {boneName=spine_01}}
     {targetBone {boneName=spine_02}}
     {targetBone {boneName=spine_03}}
     {targetBone {boneName=spine_04}}
     {targetBone {boneName=pelvis}}
    ]
    constraintWeights [1,1,1,1,1,1,1]
   }
   n54 = "Constraint - dyn_jacket_strap_l_master (PPPPPPP)"(
    ComponentPose=@n53
    Alpha=1
    ConstraintWeights_0=0.1459
    ConstraintWeights_1=0.102
    ConstraintWeights_2=0.1587
    ConstraintWeights_3=0.1325
    ConstraintWeights_4=0.0376
    ConstraintWeights_5=0.0476
    ConstraintWeights_6=0.3759
   ) {
    alphaBoolBlend {blendOption=Linear}
    boneToModify {boneName=dyn_jacket_strap_l_master}
    constraintSetup [
     {targetBone {boneName=dyn_jacket_l_01}}
     {targetBone {boneName=spine_01}}
     {targetBone {boneName=spine_02}}
     {targetBone {boneName=spine_03}}
     {targetBone {boneName=spine_04}}
     {targetBone {boneName=pelvis}}
     {targetBone {boneName=dyn_jacket_l_fr_01}}
    ]
    constraintWeights [1,1,1,1,1,1,1]
   }
   n55 = "Constraint - dyn_gun_master (PPP)"!(ComponentPose=@n54, Alpha=1, ConstraintWeights_0=0.4638, ConstraintWeights_1:float=0.4876, ConstraintWeights_2:float=0.0485) -> (Pose) {
    alphaBoolBlend {blendOption=Linear}
    boneToModify {boneName=dyn_gun_master}
    constraintSetup [{targetBone {boneName=pelvis}}, {targetBone {boneName=thigh_twist_01_r}}, {targetBone {boneName=spine_01}}]
    constraintWeights [1,1,1]
   }
   n56 = RigidBody(ComponentPose=@n55, Alpha=Get not_crouching() -> (not_crouching:double)) {
    bClampLinearTranslationLimitToRefPose=true
    bOverrideWorldGravity=true
    overridePhysicsAsset=&/M_MED_Tomcat_Casual_Acc_Physics
    simulationSpace=BaseBoneSpace
    alphaBoolBlend {blendOption=Linear}
    baseBoneRef {boneName=spine_01}
    componentAppliedLinearAccClamp=<3000 3000 3000>
    componentLinearAccScale=<0.3 0.3 0.35>
    componentLinearVelScale=<0.3 0.3 0.33>
    overrideWorldGravity {z=-980}
   }
   n57 = RigidBody(ComponentPose=@n56, Alpha=Get is_crouching()) {
    bClampLinearTranslationLimitToRefPose=true
    bOverrideWorldGravity=true
    overridePhysicsAsset=&M_MED_Tomcat_Casual_Acc_Physics
    simulationSpace=BaseBoneSpace
    alphaBoolBlend {blendOption=Linear}
    baseBoneRef {boneName=spine_01}
    componentAppliedLinearAccClamp=<3000 3000 3000>
    componentLinearAccScale=<0.3 0.3 0.35>
    componentLinearVelScale=<0.3 0.3 0.33>
    overrideWorldGravity {z=-980}
   }
   n58 = "Transform (Modify) Bone - Bone: dyn_gun_01"(ComponentPose=@n57, Alpha=Get is_crouching(), Translation="-1,0,0", Rotation="5,0,0", Scale="1,1,1") {
    rotationMode=BMM_Additive
    translationMode=BMM_Additive
    alphaBoolBlend {blendOption=Linear}
    boneToModify {boneName=dyn_gun_01}
    rotation {pitch=5}
    translation {x=-1}
   }
   n59 = "Transform (Modify) Bone - Bone: dyn_jacket_l_master"(ComponentPose=@n58, Alpha=Get is_crouching(), Translation="0,0,0.5", Rotation="10,0,0", Scale="1,1,1") {
    rotationMode=BMM_Additive
    translationMode=BMM_Additive
    alphaBoolBlend {blendOption=Linear}
    boneToModify {boneName=dyn_jacket_l_master}
    rotation {pitch=10}
    translation {z=0.5}
   }
   n60 = "Constraint - dyn_jacket_strap_r_master (PPPPPPPPPPPPP)"!(
    ComponentPose=@n59
    Alpha=1
    ConstraintWeights_0=0.1133
    ConstraintWeights_1:float=0
    ConstraintWeights_2:float=0.0025
    ConstraintWeights_3:float=0.285
    ConstraintWeights_4:float=0
    ConstraintWeights_5:float=0
    ConstraintWeights_6:float=0.0002
    ConstraintWeights_7:float=0.0969
    ConstraintWeights_8:float=0.2049
    ConstraintWeights_9:float=0.1926
    ConstraintWeights_10:float=0.0667
    ConstraintWeights_11:float=0
    ConstraintWeights_12:float=0.038
   ) -> (Pose) {
    alphaBoolBlend {blendOption=Linear}
    boneToModify {boneName=dyn_jacket_strap_r_master}
    constraintSetup [
     {targetBone {boneName=dyn_jacket_r_fr_01}}
     {targetBone {boneName=dyn_jacket_l_bk_01}}
     {targetBone {boneName=dyn_jacket_r_02}}
     {targetBone {boneName=dyn_jacket_r_01}}
     {targetBone {boneName=dyn_jacket_l_02}}
     {targetBone {boneName=dyn_jacket_r_bk_02}}
     {targetBone {boneName=dyn_jacket_r_bk_01}}
     {targetBone {boneName=spine_01}}
     {targetBone {boneName=spine_02}}
     {targetBone {boneName=spine_03}}
     {targetBone {boneName=spine_04}}
     {targetBone {boneName=spine_05}}
     {targetBone {boneName=pelvis}}
    ]
    constraintWeights [1,1,1,1,1,1,1,1,1,1,1,1,1]
   }
   n61 = "Constraint - dyn_jacket_strap_l_master (PPPPPPPPPPPPP)"!(
    ComponentPose=@n60
    Alpha=1
    ConstraintWeights_0=0.2158
    ConstraintWeights_1:float=0
    ConstraintWeights_2:float=0.0077
    ConstraintWeights_3:float=0
    ConstraintWeights_4:float=0
    ConstraintWeights_5:float=0.0447
    ConstraintWeights_6:float=0.138
    ConstraintWeights_7:float=0.1495
    ConstraintWeights_8:float=0.0672
    ConstraintWeights_9:float=0.0016
    ConstraintWeights_10:float=0.0257
    ConstraintWeights_11:float=0.0155
    ConstraintWeights_12:float=0.3343
   ) -> (Pose) {
    alphaBoolBlend {blendOption=Linear}
    boneToModify {boneName=dyn_jacket_strap_l_master}
    constraintSetup [
     {targetBone {boneName=dyn_jacket_l_01}}
     {targetBone {boneName=dyn_jacket_l_bk_01}}
     {targetBone {boneName=dyn_jacket_l_02}}
     {targetBone {boneName=dyn_jacket_r_bk_02}}
     {targetBone {boneName=dyn_jacket_l_bk_02}}
     {targetBone {boneName=spine_01}}
     {targetBone {boneName=spine_02}}
     {targetBone {boneName=spine_03}}
     {targetBone {boneName=spine_04}}
     {targetBone {boneName=spine_05}}
     {targetBone {boneName=dyn_jacket_l_fr_02}}
     {targetBone {boneName=pelvis}}
     {targetBone {boneName=dyn_jacket_l_fr_01}}
    ]
    constraintWeights [1,1,1,1,1,1,1,1,1,1,1,1,1]
   }
   n62 = "Constraint - dyn_gun_master (PPP)"(ComponentPose=@n61, Alpha=1, ConstraintWeights_0=0.4638, ConstraintWeights_1=0.4876, ConstraintWeights_2=0.0485) {
    alphaBoolBlend {blendOption=Linear}
    boneToModify {boneName=dyn_gun_master}
    constraintSetup [{targetBone {boneName=pelvis}}, {targetBone {boneName=thigh_twist_01_r}}, {targetBone {boneName=spine_01}}]
    constraintWeights [1,1,1]
   }
   n63 = "Constraint - dyn_jacket_strap_l (R)"(ComponentPose=@n62, Alpha=0.75, ConstraintWeights_0=1) {
    alphaBoolBlend {blendOption=Linear}
    boneToModify {boneName=dyn_jacket_strap_l}
    constraintSetup [{transformType=Rotation, targetBone {boneName=dyn_jacket_l_fr_01}}]
    constraintWeights [1]
   }
   n64 = "Constraint - dyn_jacket_strap_r (R)"(ComponentPose=@n63, Alpha=0.5, ConstraintWeights_0=1) {
    alphaBoolBlend {blendOption=Linear}
    boneToModify {boneName=dyn_jacket_strap_r}
    constraintSetup [{transformType=Rotation, targetBone {boneName=dyn_jacket_r_fr_01}}]
    constraintWeights [1]
   }
   n65 = "Transform (Modify) Bone - Bone: dyn_gun_master"(ComponentPose=@n64, Alpha=Get Crouch Left() -> (Crouch Left:double), Translation="0,0,0.75", Rotation="-30,0,0", Scale="1,1,1") {
    rotationMode=BMM_Additive
    rotationSpace=BCS_BoneSpace
    translationMode=BMM_Additive
    translationSpace=BCS_BoneSpace
    alphaBoolBlend {blendOption=Linear}
    boneToModify {boneName=dyn_gun_master}
    rotation {pitch=-30}
    translation {z=0.75}
   }
   Save cached pose 'Pre_Jump'(Pose=Component To Local(ComponentPose=@n65)) {cacheName=Pre_Jump}
   Use cached pose 'Pre_Jump'()
  ]}
 EventGraph {nodes [
   Event BlueprintUpdateAnimation()
   > Sequence()
   n1 = Get bIsCrouching()
   n2 = SelectFloat(A=1, B=0, bPickA=@n1)
   > Set is_crouching(is_crouching:double=@n2) -> (Output_Get:double)
   > Set gun_crouch(gun_crouch:double=float * float(A=@n2, B=0.7)) -> (Output_Get:double)
   n3 = Get bIsCrouching()
   n4 = Get bIsMoving2D()
   > Set is_crouch_walking(is_crouch_walking:double=SelectFloat(A=1, B=0, bPickA=AND Boolean(A=@n3, B=@n4))) -> (Output_Get:double)
   > Set crouch_jacket(crouch_jacket:double=SelectFloat(A=1, B=0, bPickA=Get bIsCrouchSprinting())) -> (Output_Get:double)
   > Set not crouch sprinting(not crouch sprinting:double=SelectFloat(A=0, B=1, bPickA=Get bIsCrouchSprinting())) -> (Output_Get:double)
   n5 = OR Boolean(A=Get bIsJumping(), B=Get bIsFalling())
   n6 = Get bIsSprinting()
   > Set sprint_jump(sprint_jump:double=SelectFloat(A=0, B=1, bPickA=AND Boolean(A=@n6, B=@n5))) -> (Output_Get:double)
   n7 = NOT Boolean(A=Get bIsParachuteOpened())
   n8 = Get bIsSkyDiving()
   > Set diving(diving:double=SelectFloat(A=1, B=0, bPickA=AND Boolean(A=@n8, B=@n7))) -> (Output_Get:double)
   n9 = Get bIsCrouching()
   > Set not_crouching(not_crouching:double=SelectFloat(A=1, B=0, bPickA=NOT Boolean(A=@n9))) -> (Output_Get:double)
   > Set jacket_l_Crouching(jacket_l_Crouching:double=SelectFloat(A=0, B=0.2, bPickA=Get bIsCrouching())) -> (Output_Get:double)
   > Set jacket_r_crouching(jacket_r_crouching:double=SelectFloat(A=0, B=0.7, bPickA=Get bIsCrouching())) -> (Output_Get:double)
   n10 = Get PawnMovementDirectionAzimuth() -> (PawnMovementDirectionAzimuth:float)
   n11 = float - float(A=MapRangeClamped(Value=@n10, InRangeA=-90, InRangeB=-135), B=MapRangeClamped(Value=@n10, InRangeA=-135, InRangeB=-180)) {is=title K2Node_CallFunction}
   n12 = Get bIsCrouching()
   n13 = Get bIsMoving2D()
   > Set Crouch_Bk_Left(Crouch_Bk_Left:double=SelectFloat(A="Absolute (Float)"(A=@n11), B=0, bPickA=AND Boolean(A=@n12, B=@n13))) -> (Output_Get:double)
   n14 = Get PawnMovementDirectionAzimuth()
   > Set Crouch_Bk(Crouch_Bk:double=float + float(
     A=MapRangeClamped(Value=@n14, InRangeA=-135, InRangeB=-180, OutRangeB=0.75)
     B=MapRangeClamped(Value=@n14, InRangeA=135, InRangeB=180, OutRangeB=0.75)
    ) {is=title K2Node_CommutativeAssociativeBinaryOperator}) -> (Output_Get:double)
   n15 = Get PawnMovementDirectionAzimuth()
   n16 = "Absolute (Float)"(A=float - float(A=MapRangeClamped(Value=@n15, InRangeA=-45, InRangeB=-90), B=MapRangeClamped(Value=@n15, InRangeA=-90, InRangeB=-135)))
   n17 = Get bIsCrouching()
   n18 = Get bIsMoving2D()
   > Set Crouch Left(Crouch Left:double=SelectFloat(A=@n16, B=0, bPickA=AND Boolean(A=@n17, B=@n18))) -> (Output_Get:double)
   n19 = GetAnimInstance(self=GetOwnerSkeletalMeshComponent())
   > Is Valid(InputObject=@n19)
   > Set is_emote_playing(is_emote_playing:double=SelectFloat(
     A=1
     B=0
     bPickA=Montage_IsActive!(
      self:Anim Instance Object Reference=@n19
      Montage:Anim Montage Object Reference=/BRCosmetics/Animation/Game/MainPlayer/Emotes/BlackMondayFight/Emote_BlackMondayFight_CMM_M
     ) -> (ReturnValue:bool)
    )) -> (Output_Get:double)
   Event OnEndPlayingEmote() {is=title K2Node_Event}
   > Delay(Duration=0.2)
   > ResetDynamics(InTeleportType=None)
   MapRangeClamped(Value=@n16, InRangeA=0, InRangeB=1, OutRangeA=1, OutRangeB=0)
   AND Boolean(A=@n1, B=NOT Boolean(A=Get bIsMoving2D()))
   Get bIsTargeting()
  ]}
]
variables [
 Crouch Left {type=number}
 Crouch_Bk {type=number}
 Crouch_Bk_Left {type=number}
 crouch_jacket {type=number}
 diving {type=number}
 gun_crouch {type=number}
 is_crouch_walking {type=number}
 is_crouching {type=number}
 is_emote_playing {type=number}
 jacket_l_Crouching {type=number}
 jacket_r_crouching {type=number}
 not crouch sprinting {type=number}
 not_crouching {type=number}
 sprint_jump {type=number}
]
